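# UI material for item shop tiles in the EmeraldGlass art set, blending a background texture toward a fixed red color.
# The mask remaps UVs around the center and takes the absolute vertical component, giving symmetric bands at the top and bottom edges.
# Subtract, Falloff and Intensity parameters shape the bands.
# Intensity multiplies after the saturate, so the lerp alpha is unclamped and can exceed one.
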Material /ItemShop/UI/Art/EmeraldGlass/M_UI_EmeraldGlass_ShopTiles
graph {
 properties {bUsedWithStaticMesh=false, materialDomain=MD_UI, pixelDepthOffsetMode=PDOM_Legacy, usedShadingModels=Default Lit}
 nodes [
  n1 = Abs(None=ConstantBiasScale(None=TextureCoordinate()) {bias=-0.5, scale=2})
  n2 = Multiply(
   A=Subtract(A=ComponentMask(None=@n1) {g=true}, B=ScalarParameter() {defaultValue=0.614, parameterName=Subtract})
   B=ScalarParameter() {defaultValue=1.1505, parameterName=Falloff}
  )
  LinearInterpolate(
   A=TextureSample() {texture=&/T_UI_EmeraldGlass_ShopBG}.RGB
   B=Constant3Vector() {constant=<rgba 0.6823 0.0151 0.0002 1>}
   Alpha=Multiply(A=Saturate(None=@n2), B=ScalarParameter() {defaultValue=1.7562, parameterName=Intensity})
  )
 ]
}
