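# UI-domain material for Emerald Glass item shop tiles, with the graph showing no connection to material outputs.
# One branch builds a rounded outline from shared MPC_ItemShop corner settings, filled with a blue or gold gradient via bIsBlue.
# Another blends red to gold through a smoothstepped BoxSkew shape, scaled by aspect ratios from UV channel 3.
# In that branch both bIsBlue inputs are identical gold gradients.
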Material /ItemShop/UI/Art/EmeraldGlass/M_UI_EmeraldGlass_ShopTiles_02
graph {
 properties {bUsedWithStaticMesh=false, materialDomain=MD_UI, pixelDepthOffsetMode=PDOM_Legacy, usedShadingModels=Default Lit}
 nodes [
  n1 = Constant3Vector() {constant=<rgba 0.776 0.23 0 1>}
  n2 = Constant3Vector() {constant=<rgba 0.9844 0.7538 0.0345 1>}
  n3 = MaterialFunctionCall(GradientStart=Constant() {r=-0.144}, GradientEnd=Constant() {r=0.9047}, RotationAngle=Constant() {r=0.25}) {materialFunction=&MF_UI_LinearGradient}
  n4 = Constant3Vector() {constant=<rgba 0.776 0.23 0 1>}
  n5 = Constant3Vector() {constant=<rgba 0.9844 0.7538 0.0345 1>}
  n6 = MaterialFunctionCall(GradientStart=Constant() {r=-0.144}, GradientEnd=Constant() {r=0.9047}, RotationAngle=Constant() {r=0.25}) {materialFunction=&MF_UI_LinearGradient}
  n7 = ScalarParameter() {defaultValue=0.01, parameterName=Smoothness}
  n8 = NamedRerouteUsage()
  n9 = Constant()
  n10 = Constant()
  n11 = Constant()
  n12 = Constant()
  n13 = Constant()
  n14 = ScalarParameter() {defaultValue=0.2, group=Inner Skew Box, parameterName=I_SkewScale}
  n15 = TextureCoordinate() {coordinateIndex=3}
  n16 = ComponentMask(None=@n15) {g=true}
  n17 = ComponentMask(None=@n15) {r=true}
  n18 = Divide(A=@n16, B=@n17)
  n19 = Divide(A=@n17, B=@n16)
  n20 = Saturate(None=AppendVector(A=@n18, B=@n19))
  n21 = Multiply(A=OneMinus(None=Multiply(A=@n14, B=@n20))) {constB=1.3398}
  n22 = StaticBool()
  LinearInterpolate(
   A=Constant3Vector() {constant=<rgba 0.9844 0 0.105 1>}
   B=StaticSwitchParameter(True=LinearInterpolate(A=@n1, B=@n2, Alpha=@n3.Result), False=LinearInterpolate(A=@n4, B=@n5, Alpha=@n6.Result)) {parameterName=bIsBlue}
   Alpha=SmoothStep(
    Max=Multiply(A=@n7, B=@n8)
    Value=MaterialFunctionCall!(MinWidth=@n9, Angle Top=@n10, Angle Right=@n11, Angle Bottom=@n12, Angle Left=@n13, Scale=@n21, Flip=@n22) {materialFunction=/FortUILibrary/MaterialFunctions/Shapes/MF_UI_BoxSkew}.Result
   )
  ) {constB=0}
  n23 = LinearInterpolate(
   A=Constant3Vector() {constant=<rgba 0.0168 0.0482 0.552 1>}
   B=Constant3Vector() {constant=<rgba 0.0782 0.5972 0.9911 1>}
   Alpha=MaterialFunctionCall(GradientStart=Constant() {r=-0.144}, GradientEnd=Constant() {r=0.9047}, RotationAngle=Constant() {r=0.25}) {materialFunction=&MF_UI_LinearGradient}.Result
  )
  n24 = LinearInterpolate(
   A=Constant3Vector() {constant=<rgba 0.776 0.23 0 1>}
   B=Constant3Vector() {constant=<rgba 0.9844 0.7538 0.0345 1>}
   Alpha=MaterialFunctionCall(GradientStart=Constant() {r=-0.144}, GradientEnd=Constant() {r=0.9047}, RotationAngle=Constant() {r=0.25}) {materialFunction=&MF_UI_LinearGradient}.Result
  )
  n25 = MaterialFunctionCall(
   Corner Radius=CollectionParameter() {collection=/FortUILibrary/Systems/ItemShopTile/MPC_ItemShop, parameterName=CornerRadius}
   Radius in Pixels=CollectionParameter() {collection=&MPC_ItemShop, parameterName=bRadiusInPixels}
   Outline Thickness=ScalarParameter() {defaultValue=6, parameterName=Border Thickness}
  ) {materialFunction=&MF_UI_CoreUI_RoundedMask}
  LinearInterpolate(
   A=Constant()
   B=LinearInterpolate(B=StaticSwitchParameter(True=@n23, False=@n24) {parameterName=bIsBlue}, Alpha=@n25.Shape Mask)
   Alpha=SmoothStep(Value=@n25.Outline SDF) {constMax=0.008}
  )
  MaterialFunctionCall() {materialFunction=&MF_UI_UVScaling}
  NamedRerouteDeclaration(None=@n18) {name=ScaleX, nodeColor=<rgba 0 0 0 1>}
  NamedRerouteDeclaration(None=@n19) {name=ScaleY, nodeColor=<rgba 0 0 0 1>}
  NamedRerouteUsage()
 ]
}
